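# SoundCue that supplies the looping TMNT instrumental music for the Helios Underground Lair music party event.
# A single looping wave player is the root, played at reduced volume with a very large max distance and its own concurrency cap.
# The base submix is disabled and the signal goes at full send level to the MusicParty Underground Lair audio bus, so the cue acts as a bus feed rather than direct output.
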SoundCue /Game/Sounds/Helios/POI/UndergroundLair/Cue_MusicParty_Helios_UndergroundLair_Event
root=WavePlayer_2
nodes [{id=WavePlayer_2, properties {bLooping=true, soundWaveAssetPtr=&/s28_Fortnite_TMNT_Instrumental_Loop}}]
properties {
 bEnableBaseSubmix=false
 bOverrideConcurrency=true
 maxDistance=2097152
 volumeMultiplier=0.5
 busSends [{audioBus=&/MusicParty_MS_Helios_UndergroundLair, minSendLevel=0, sendLevel=1, soundSourceBus=null}]
 concurrencyOverrides {maxCount=4}
}
